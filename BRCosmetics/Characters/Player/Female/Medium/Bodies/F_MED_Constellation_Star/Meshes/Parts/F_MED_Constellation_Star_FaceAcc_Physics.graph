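# Physics asset for the FaceAcc part of the female medium Constellation Star body, setting up dynamic hair strands and curls.
# Capsules cover head, neck, upper spine, clavicles and upper arms, while tiny MSTR bodies anchor hair chains to the head.
# Chains use narrow swing and twist limits, generally looser toward the ends, with curl tips branching off.
# Free-rotation cross-constraints link neighboring back and middle strand ends sideways.
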PhysicsAsset /BRCosmetics/Characters/Player/Female/Medium/Bodies/F_MED_Constellation_Star/Meshes/Parts/F_MED_Constellation_Star_FaceAcc_Physics
bodies {
 clavicle_l [{shapeType=Capsule, length=7, radius=5, center=<3.8857 -0.0001 0>, rotation=<rot 90 0.0395 0.0395>}]
 clavicle_r [{shapeType=Capsule, length=7, radius=5, center=<-3.8857 0.0001 0>, rotation=<rot 90 -0.0198 179.9802>}]
 dyn_hair_bk_00 [{shapeType=Capsule, length=0.505, radius=0.505, shapeName=dyn_hair_bk_00_capsule}]
 dyn_hair_bk_00_MSTR [{shapeType=Capsule, length=0.01, radius=0.01}]
 dyn_hair_bk_00_l [{shapeType=Capsule, length=0.505, radius=0.505, shapeName=dyn_hair_bk_00_l_capsule}]
 dyn_hair_bk_00_l_MSTR [{shapeType=Capsule, length=0.01, radius=0.01}]
 dyn_hair_bk_00_r [{shapeType=Capsule, length=0.505, radius=0.505, shapeName=dyn_hair_bk_00_r_capsule}]
 dyn_hair_bk_00_r_MSTR [{shapeType=Capsule, length=0.01, radius=0.01}]
 dyn_hair_bk_01 [{shapeType=Capsule, length=10, radius=2.5, shapeName=dyn_hair_bk_01_capsule, center=<3.75 0 -0.2>, rotation=<rot 0 0 90>}]
 dyn_hair_bk_01_l [{shapeType=Capsule, length=10, radius=2.5, shapeName=dyn_hair_bk_01_l_capsule, center=<3.75 0 0>, rotation=<rot 0 0 90>}]
 dyn_hair_bk_01_r [{shapeType=Capsule, length=10, radius=2.5, shapeName=dyn_hair_bk_01_r_capsule, center=<3.75 0 0>, rotation=<rot 0 0 90>}]
 dyn_hair_bk_02 [{shapeType=Capsule, length=7, radius=2.5, shapeName=dyn_hair_bk_02_capsule, center=<3.75 0 -1>, rotation=<rot 0 0 90>}]
 dyn_hair_bk_02_l [{shapeType=Capsule, length=7, radius=2.5, shapeName=dyn_hair_bk_02_l_capsule, center=<3.75 0 -0.75>, rotation=<rot 0 0 90>}]
 dyn_hair_bk_02_r [{shapeType=Capsule, length=7, radius=2.5, shapeName=dyn_hair_bk_02_r_capsule, center=<3.75 0 -0.75>, rotation=<rot 0 0 90>}]
 dyn_hair_bk_03 [{shapeType=Capsule, length=7, radius=2, shapeName=dyn_hair_bk_03_capsule, center=<2.8 0 0>, rotation=<rot 0 0 90>}]
 dyn_hair_bk_03_l [{shapeType=Capsule, length=7, radius=3, shapeName=dyn_hair_bk_03_l_capsule, center=<2.8 0 0>, rotation=<rot 0 0 60>}]
 dyn_hair_bk_03_r [{shapeType=Capsule, length=7, radius=3, shapeName=dyn_hair_bk_03_r_capsule, center=<2.8 0.5626 0.9744>, rotation=<rot 0 0 120>}]
 dyn_hair_fr_00_l [{shapeType=Capsule, length=0.505, radius=0.505, shapeName=dyn_hair_fr_00_l_capsule}]
 dyn_hair_fr_00_l_MSTR [{shapeType=Capsule, length=0.01, radius=0.01}]
 dyn_hair_fr_00_r [{shapeType=Capsule, length=0.505, radius=0.505, shapeName=dyn_hair_fr_00_r_capsule}]
 dyn_hair_fr_00_r_MSTR [{shapeType=Capsule, length=0.01, radius=0.01}]
 dyn_hair_fr_01_l [{shapeType=Capsule, length=10, radius=2.5, shapeName=dyn_hair_fr_01_l_capsule, center=<3.75 0 2.4>, rotation=<rot 0 0 90>}]
 dyn_hair_fr_01_r [{shapeType=Capsule, length=10, radius=2.5, shapeName=dyn_hair_fr_01_r_capsule, center=<3.75 0 2.4>, rotation=<rot 0 0 90>}]
 dyn_hair_fr_02_l [{shapeType=Capsule, length=7, radius=2.5, shapeName=dyn_hair_fr_02_l_capsule, center=<3.75 0 1.7>, rotation=<rot 0 0 90>}]
 dyn_hair_fr_02_r [{shapeType=Capsule, length=7, radius=2.5, shapeName=dyn_hair_fr_02_r_capsule, center=<3.75 0 1.7>, rotation=<rot 0 0 90>}]
 dyn_hair_fr_03_l [{shapeType=Capsule, length=2, radius=2, shapeName=dyn_hair_fr_03_l_capsule, center=<2.8 0 0>, rotation=<rot 0 0 90>}]
 dyn_hair_fr_03_r [{shapeType=Capsule, length=2, radius=2, shapeName=dyn_hair_fr_03_r_capsule, center=<2.8 0 0>, rotation=<rot 0 0 90>}]
 dyn_hair_md_00_l [{shapeType=Capsule, length=0.505, radius=0.505, shapeName=dyn_hair_md_00_l_capsule}]
 dyn_hair_md_00_l_MSTR [{shapeType=Capsule, length=0.01, radius=0.01}]
 dyn_hair_md_00_r [{shapeType=Capsule, length=0.505, radius=0.505, shapeName=dyn_hair_md_00_r_capsule}]
 dyn_hair_md_00_r_MSTR [{shapeType=Capsule, length=0.01, radius=0.01}]
 dyn_hair_md_01_l [{shapeType=Capsule, length=10, radius=2.5, shapeName=dyn_hair_md_01_l_capsule, center=<3.75 0 2.4>, rotation=<rot 0 0 90>}]
 dyn_hair_md_01_r [{shapeType=Capsule, length=10, radius=2.5, shapeName=dyn_hair_md_01_r_capsule, center=<3.75 0 2.4>, rotation=<rot 0 0 90>}]
 dyn_hair_md_02_l [{shapeType=Capsule, length=7, radius=2.5, shapeName=dyn_hair_md_02_l_capsule, center=<3.75 0 1.7>, rotation=<rot 0 0 90>}]
 dyn_hair_md_02_r [{shapeType=Capsule, length=7, radius=2.5, shapeName=dyn_hair_md_02_r_capsule, center=<3.75 0 1.7>, rotation=<rot 0 0 90>}]
 dyn_hair_md_03_l [{shapeType=Capsule, length=2, radius=2, shapeName=dyn_hair_md_03_l_capsule, center=<2.8 0 0>, rotation=<rot 0 0 90>}]
 dyn_hair_md_03_r [{shapeType=Capsule, length=2, radius=2, shapeName=dyn_hair_md_03_r_capsule, center=<2.8 0 0>, rotation=<rot 0 0 90>}]
 dyn_haircurl_bt_00_l [{shapeType=Capsule, length=3, radius=1, shapeName=dyn_haircurl_bt_00_l_capsule, center=<1 0 0>, rotation=<rot 0 0 90>}]
 dyn_haircurl_fr_tp_00 [{shapeType=Capsule, length=7, radius=2.25, shapeName=dyn_haircurl_fr_tp_00_capsule, center=<2.7 0 0>, rotation=<rot 0 0 90>}]
 dyn_haircurl_fr_tp_00_MSTR [{shapeType=Capsule, length=0.01, radius=0.01}]
 dyn_haircurl_fr_tp_01 [{shapeType=Capsule, length=7, radius=2.25, shapeName=dyn_haircurl_fr_tp_01_capsule, center=<5.7 0 0>, rotation=<rot 0 0 90>}]
 dyn_haircurl_tp_00_r_MSTR [{shapeType=Capsule, length=0.01, radius=0.01}]
 dyn_haircurl_tp_01_r [{shapeType=Capsule, length=2.5, radius=0.5, shapeName=dyn_haircurl_tp_01_r_capsule, center=<1 0 0>, rotation=<rot 0 0 90>}]
 dyn_haircurltip_010_00 [{shapeType=Capsule, length=3, radius=1, shapeName=dyn_haircurltip_010_00_capsule, center=<1 0 0>, rotation=<rot 0 0 90>}]
 dyn_haircurltip_01_00 [{shapeType=Capsule, length=3, radius=1, shapeName=dyn_haircurltip_01_00_capsule, center=<1 0 0>, rotation=<rot 0 0 90>}]
 dyn_haircurltip_02_00 [{shapeType=Capsule, length=3, radius=1, shapeName=dyn_haircurltip_02_00_capsule, center=<1 0 0>, rotation=<rot 0 0 90>}]
 dyn_haircurltip_03_00 [{shapeType=Capsule, length=3, radius=1, shapeName=dyn_haircurltip_03_00_capsule, center=<1 0 0>, rotation=<rot 0 0 90>}]
 dyn_haircurltip_04_00 [{shapeType=Capsule, length=3, radius=1, shapeName=dyn_haircurltip_04_00_capsule, center=<1 0 0>, rotation=<rot 0 0 90>}]
 dyn_haircurltip_05_00 [{shapeType=Capsule, length=3, radius=1, shapeName=dyn_haircurltip_05_00_capsule, center=<1 0 0>, rotation=<rot 0 0 90>}]
 dyn_haircurltip_06_00 [{shapeType=Capsule, length=3, radius=1, shapeName=dyn_haircurltip_06_00_capsule, center=<1 0 0>, rotation=<rot 0 0 90>}]
 dyn_haircurltip_07_00 [{shapeType=Capsule, length=3, radius=1, shapeName=dyn_haircurltip_07_00_capsule, center=<1.8517 0 -0.1661>, rotation=<rot 0 0 90>}]
 dyn_haircurltip_08_00 [{shapeType=Capsule, length=3, radius=1, shapeName=dyn_haircurltip_08_00_capsule, center=<1 0 0>, rotation=<rot 0 0 90>}]
 dyn_haircurltip_09_00 [{shapeType=Capsule, length=3, radius=1, shapeName=dyn_haircurltip_09_00_capsule, center=<1 0 0>, rotation=<rot 0 0 90>}]
 head [
  {shapeType=Capsule, length=6.5, radius=5.75, center=<3.5617 0 0>, rotation=<rot -90 -0.0074 0.0074>}
  {shapeType=Capsule, length=8, radius=4.25, center=<2.3241 2.6329 0>, rotation=<rot 90 -1.4321 8.569>}
 ]
 neck_01 [{shapeType=Capsule, length=7, radius=5.5, rotation=<rot -90 0.0049 -0.0049>}]
 spine_05 [{shapeType=Capsule, length=7, radius=10.75}]
 upperarm_l [{shapeType=Capsule, length=7, radius=5, center=<3.8857 -0.0001 0>, rotation=<rot 90 0.0395 0.0395>}]
 upperarm_r [{shapeType=Capsule, length=7, radius=5, center=<-3.8857 0.0001 0>, rotation=<rot 90 -0.0198 179.9802>}]
}
constraints [
 {
  bone1Name=head
  bone2Name=neck_01
  swing1LimitDegrees=45
  swing1Motion=Free
  swing2LimitDegrees=45
  swing2Motion=Free
  twistLimitDegrees=45
  twistMotion=Free
 }
 {
  bone1Name=neck_01
  bone2Name=spine_05
  swing1LimitDegrees=45
  swing1Motion=Free
  swing2LimitDegrees=45
  swing2Motion=Free
  twistLimitDegrees=45
  twistMotion=Free
 }
 {
  bone1Name=dyn_haircurl_tp_00_r_MSTR
  bone2Name=head
  swing1LimitDegrees=45
  swing1Motion=Free
  swing2LimitDegrees=45
  swing2Motion=Free
  twistLimitDegrees=45
  twistMotion=Free
 }
 {
  bone1Name=dyn_haircurl_fr_tp_00_MSTR
  bone2Name=head
  swing1LimitDegrees=45
  swing1Motion=Free
  swing2LimitDegrees=45
  swing2Motion=Free
  twistLimitDegrees=45
  twistMotion=Free
 }
 {
  bone1Name=dyn_hair_fr_00_l_MSTR
  bone2Name=head
  swing1LimitDegrees=45
  swing1Motion=Free
  swing2LimitDegrees=45
  swing2Motion=Free
  twistLimitDegrees=45
  twistMotion=Free
 }
 {
  bone1Name=dyn_hair_fr_00_r_MSTR
  bone2Name=head
  swing1LimitDegrees=45
  swing1Motion=Free
  swing2LimitDegrees=45
  swing2Motion=Free
  twistLimitDegrees=45
  twistMotion=Free
 }
 {
  bone1Name=dyn_hair_md_00_l_MSTR
  bone2Name=head
  swing1LimitDegrees=45
  swing1Motion=Free
  swing2LimitDegrees=45
  swing2Motion=Free
  twistLimitDegrees=45
  twistMotion=Free
 }
 {
  bone1Name=dyn_hair_bk_00_l_MSTR
  bone2Name=head
  swing1LimitDegrees=45
  swing1Motion=Free
  swing2LimitDegrees=45
  swing2Motion=Free
  twistLimitDegrees=45
  twistMotion=Free
 }
 {
  bone1Name=dyn_hair_bk_00_MSTR
  bone2Name=head
  swing1LimitDegrees=45
  swing1Motion=Free
  swing2LimitDegrees=45
  swing2Motion=Free
  twistLimitDegrees=45
  twistMotion=Free
 }
 {
  bone1Name=dyn_hair_bk_00_r_MSTR
  bone2Name=head
  swing1LimitDegrees=45
  swing1Motion=Free
  swing2LimitDegrees=45
  swing2Motion=Free
  twistLimitDegrees=45
  twistMotion=Free
 }
 {
  bone1Name=dyn_hair_md_00_r_MSTR
  bone2Name=head
  swing1LimitDegrees=45
  swing1Motion=Free
  swing2LimitDegrees=45
  swing2Motion=Free
  twistLimitDegrees=45
  twistMotion=Free
 }
 {bone1Name=dyn_hair_bk_00, bone2Name=dyn_hair_bk_00_MSTR, swing1LimitDegrees=5, swing2LimitDegrees=5, twistLimitDegrees=5}
 {bone1Name=dyn_hair_bk_01, bone2Name=dyn_hair_bk_00, swing1LimitDegrees=5, swing2LimitDegrees=5, twistLimitDegrees=1}
 {bone1Name=dyn_hair_bk_02, bone2Name=dyn_hair_bk_01, swing1LimitDegrees=10, swing2LimitDegrees=10, twistLimitDegrees=5}
 {bone1Name=dyn_hair_bk_03, bone2Name=dyn_hair_bk_02, swing1LimitDegrees=10, swing2LimitDegrees=30, twistLimitDegrees=5}
 {bone1Name=dyn_hair_bk_00_l, bone2Name=dyn_hair_bk_00_l_MSTR, swing1LimitDegrees=5, swing2LimitDegrees=5, twistLimitDegrees=5}
 {bone1Name=dyn_hair_bk_01_l, bone2Name=dyn_hair_bk_00_l, swing1LimitDegrees=5, swing2LimitDegrees=5, twistLimitDegrees=1}
 {bone1Name=dyn_hair_bk_02_l, bone2Name=dyn_hair_bk_01_l, swing1LimitDegrees=10, swing2LimitDegrees=10, twistLimitDegrees=5}
 {bone1Name=dyn_hair_bk_03_l, bone2Name=dyn_hair_bk_02_l, swing1LimitDegrees=30, swing2LimitDegrees=30, twistLimitDegrees=5}
 {bone1Name=dyn_hair_md_01_l, bone2Name=dyn_hair_md_00_l, swing1LimitDegrees=5, swing2LimitDegrees=5, twistLimitDegrees=1}
 {bone1Name=dyn_hair_md_02_l, bone2Name=dyn_hair_md_01_l, swing1LimitDegrees=10, swing2LimitDegrees=10, twistLimitDegrees=5}
 {bone1Name=dyn_hair_md_00_l, bone2Name=dyn_hair_md_00_l_MSTR, swing1LimitDegrees=5, swing2LimitDegrees=5, twistLimitDegrees=5}
 {bone1Name=dyn_hair_md_03_l, bone2Name=dyn_hair_md_02_l, swing1LimitDegrees=20, swing2LimitDegrees=20, twistLimitDegrees=5}
 {bone1Name=dyn_hair_fr_00_r, bone2Name=dyn_hair_fr_00_r_MSTR, swing1LimitDegrees=10, swing2LimitDegrees=10, twistLimitDegrees=5}
 {bone1Name=dyn_hair_fr_01_r, bone2Name=dyn_hair_fr_00_r, swing1LimitDegrees=5, swing2LimitDegrees=5, twistLimitDegrees=1}
 {bone1Name=dyn_hair_fr_02_r, bone2Name=dyn_hair_fr_01_r, swing1LimitDegrees=10, swing2LimitDegrees=10, twistLimitDegrees=5}
 {bone1Name=dyn_hair_fr_03_r, bone2Name=dyn_hair_fr_02_r, swing1LimitDegrees=10, swing2LimitDegrees=30, twistLimitDegrees=5}
 {bone1Name=dyn_hair_fr_00_l, bone2Name=dyn_hair_fr_00_l_MSTR, swing1LimitDegrees=10, swing2LimitDegrees=10, twistLimitDegrees=5}
 {bone1Name=dyn_hair_fr_01_l, bone2Name=dyn_hair_fr_00_l, swing1LimitDegrees=5, swing2LimitDegrees=5, twistLimitDegrees=1}
 {bone1Name=dyn_hair_fr_02_l, bone2Name=dyn_hair_fr_01_l, swing1LimitDegrees=10, swing2LimitDegrees=10, twistLimitDegrees=5}
 {bone1Name=dyn_hair_fr_03_l, bone2Name=dyn_hair_fr_02_l, swing1LimitDegrees=10, swing2LimitDegrees=30, twistLimitDegrees=5}
 {bone1Name=dyn_hair_bk_00_r, bone2Name=dyn_hair_bk_00_r_MSTR, swing1LimitDegrees=5, swing2LimitDegrees=5, twistLimitDegrees=5}
 {bone1Name=dyn_hair_bk_01_r, bone2Name=dyn_hair_bk_00_r, swing1LimitDegrees=5, swing2LimitDegrees=5, twistLimitDegrees=1}
 {bone1Name=dyn_hair_bk_02_r, bone2Name=dyn_hair_bk_01_r, swing1LimitDegrees=10, swing2LimitDegrees=10, twistLimitDegrees=5}
 {bone1Name=dyn_hair_bk_03_r, bone2Name=dyn_hair_bk_02_r, swing1LimitDegrees=30, swing2LimitDegrees=30, twistLimitDegrees=5}
 {bone1Name=dyn_hair_md_00_r, bone2Name=dyn_hair_md_00_r_MSTR, swing1LimitDegrees=5, swing2LimitDegrees=5, twistLimitDegrees=5}
 {bone1Name=dyn_hair_md_01_r, bone2Name=dyn_hair_md_00_r, swing1LimitDegrees=5, swing2LimitDegrees=5, twistLimitDegrees=1}
 {bone1Name=dyn_hair_md_02_r, bone2Name=dyn_hair_md_01_r, swing1LimitDegrees=10, swing2LimitDegrees=10, twistLimitDegrees=5}
 {bone1Name=dyn_hair_md_03_r, bone2Name=dyn_hair_md_02_r, swing1LimitDegrees=20, swing2LimitDegrees=20, twistLimitDegrees=5}
 {
  bone1Name=clavicle_l
  bone2Name=spine_05
  swing1LimitDegrees=45
  swing1Motion=Free
  swing2LimitDegrees=45
  swing2Motion=Free
  twistLimitDegrees=45
  twistMotion=Free
 }
 {
  bone1Name=upperarm_l
  bone2Name=clavicle_l
  swing1LimitDegrees=45
  swing1Motion=Free
  swing2LimitDegrees=45
  swing2Motion=Free
  twistLimitDegrees=45
  twistMotion=Free
 }
 {
  bone1Name=upperarm_r
  bone2Name=clavicle_r
  swing1LimitDegrees=45
  swing1Motion=Free
  swing2LimitDegrees=45
  swing2Motion=Free
  twistLimitDegrees=45
  twistMotion=Free
 }
 {
  bone1Name=clavicle_r
  bone2Name=spine_05
  swing1LimitDegrees=45
  swing1Motion=Free
  swing2LimitDegrees=45
  swing2Motion=Free
  twistLimitDegrees=45
  twistMotion=Free
 }
 {bone1Name=dyn_haircurl_fr_tp_01, bone2Name=dyn_haircurl_fr_tp_00, swing1LimitDegrees=10, swing2LimitDegrees=30, twistLimitDegrees=2}
 {bone1Name=dyn_haircurl_fr_tp_00, bone2Name=dyn_haircurl_fr_tp_00_MSTR, swing1LimitDegrees=5, swing2LimitDegrees=30, twistLimitDegrees=2}
 {bone1Name=dyn_haircurl_tp_01_r, bone2Name=dyn_haircurl_tp_00_r_MSTR, swing1LimitDegrees=30, swing2LimitDegrees=30, twistLimitDegrees=15}
 {bone1Name=dyn_haircurltip_010_00, bone2Name=dyn_hair_fr_03_r, swing1LimitDegrees=10, swing2LimitDegrees=10, twistLimitDegrees=15}
 {bone1Name=dyn_haircurltip_02_00, bone2Name=dyn_hair_md_03_l, swing1LimitDegrees=10, swing2LimitDegrees=10, twistLimitDegrees=15}
 {bone1Name=dyn_haircurltip_01_00, bone2Name=dyn_hair_md_02_l, swing1LimitDegrees=10, swing2LimitDegrees=10, twistLimitDegrees=15}
 {bone1Name=dyn_haircurltip_04_00, bone2Name=dyn_hair_bk_03_l, swing1LimitDegrees=10, swing2LimitDegrees=10, twistLimitDegrees=15}
 {bone1Name=dyn_haircurltip_03_00, bone2Name=dyn_hair_bk_03_l, swing1LimitDegrees=10, swing2LimitDegrees=10, twistLimitDegrees=15}
 {bone1Name=dyn_haircurl_bt_00_l, bone2Name=dyn_hair_bk_02_l, swing1LimitDegrees=10, swing2LimitDegrees=10, twistLimitDegrees=15}
 {bone1Name=dyn_haircurltip_05_00, bone2Name=dyn_hair_bk_03, swing1LimitDegrees=10, swing2LimitDegrees=10, twistLimitDegrees=15}
 {bone1Name=dyn_haircurltip_07_00, bone2Name=dyn_hair_bk_03_r, swing1LimitDegrees=10, swing2LimitDegrees=10, twistLimitDegrees=15}
 {bone1Name=dyn_haircurltip_06_00, bone2Name=dyn_hair_bk_03_r, swing1LimitDegrees=10, swing2LimitDegrees=10, twistLimitDegrees=15}
 {bone1Name=dyn_haircurltip_08_00, bone2Name=dyn_hair_bk_03_r, swing1LimitDegrees=10, swing2LimitDegrees=10, twistLimitDegrees=15}
 {bone1Name=dyn_haircurltip_09_00, bone2Name=dyn_hair_md_03_r, swing1LimitDegrees=10, swing2LimitDegrees=10, twistLimitDegrees=15}
 {
  bone1Name=dyn_hair_bk_03_r
  bone2Name=dyn_hair_bk_03
  swing1LimitDegrees=45
  swing1Motion=Free
  swing2LimitDegrees=45
  swing2Motion=Free
  twistLimitDegrees=45
  twistMotion=Free
 }
 {
  bone1Name=dyn_hair_bk_03_l
  bone2Name=dyn_hair_bk_03
  swing1LimitDegrees=45
  swing1Motion=Free
  swing2LimitDegrees=45
  swing2Motion=Free
  twistLimitDegrees=45
  twistMotion=Free
 }
 {
  bone1Name=dyn_hair_md_03_l
  bone2Name=dyn_hair_bk_03_l
  swing1LimitDegrees=45
  swing1Motion=Free
  swing2LimitDegrees=45
  swing2Motion=Free
  twistLimitDegrees=45
  twistMotion=Free
 }
 {
  bone1Name=dyn_hair_md_03_r
  bone2Name=dyn_hair_bk_03_r
  swing1LimitDegrees=45
  swing1Motion=Free
  swing2LimitDegrees=45
  swing2Motion=Free
  twistLimitDegrees=45
  twistMotion=Free
 }
]
