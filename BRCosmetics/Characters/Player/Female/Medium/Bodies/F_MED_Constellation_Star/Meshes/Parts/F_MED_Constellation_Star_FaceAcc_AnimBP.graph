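# Post-copy animation Blueprint for the hair on the Constellation Star female face accessory mesh, adding aim and physics motion.
# The parent mesh pose passes through middle and back hair Look At bones, neck-space RigidBody simulation and a small bone offset.
# Look At eases out during emotes.
# RigidBody gravity switches to scaled skydiving force while skydiving; front hair Look At nodes are unconnected.
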AnimBlueprint /BRCosmetics/Characters/Player/Female/Medium/Bodies/F_MED_Constellation_Star/Meshes/Parts/F_MED_Constellation_Star_FaceAcc_AnimBP
target_skeleton=/Game/Characters/Player/Female/Medium/Bodies/F_MED_Constellation_Star/Meshes/Parts/F_MED_Constellation_Star_FaceAcc_Skeleton
graphs [
 AnimGraph {nodes [
   n1 = Look At - Bone: dyn_hair_fr_00_l_JNT(Alpha=0) {
    alphaBoolBlend {blendOption=Linear}
    boneToModify {boneName=dyn_hair_fr_00_l_JNT}
    lookAtLocation {x=0}
    lookAtTarget {boneReference {boneName=dyn_hair_fr_00_l_JNT_aim}}
    lookAt_Axis {axis=<1 0>}
    lookUp_Axis {axis {x=0, z=1}}
   }
   Look At - Bone: dyn_hair_fr_00_r_JNT(ComponentPose=@n1, Alpha=0) {
    alphaBoolBlend {blendOption=Linear}
    boneToModify {boneName=dyn_hair_fr_00_r_JNT}
    lookAtLocation {x=0}
    lookAtTarget {boneReference {boneName=dyn_hair_fr_00_r_JNT_aim}}
    lookAt_Axis {axis=<1 0>}
    lookUp_Axis {axis {x=0, z=1}}
   }
   n2 = Get LookAtAlpha() -> (LookAtAlpha:double)
   n3 = Look At - Bone: dyn_hair_md_00_l_JNT(ComponentPose=Local To Component(LocalPose=Copy Pose From Mesh() {bUseAttachedParent=true}), Alpha=@n2) {
    alphaBoolBlend {blendOption=Linear}
    boneToModify {boneName=dyn_hair_md_00_l_JNT}
    lookAtLocation {x=0}
    lookAtTarget {boneReference {boneName=dyn_hair_md_00_l_JNT_aim}}
    lookAt_Axis {axis=<1 0>}
    lookUp_Axis {axis {x=0, z=1}}
   }
   n4 = Look At - Bone: dyn_hair_bk_00_l_JNT(ComponentPose=@n3, Alpha=@n2) {
    alphaBoolBlend {blendOption=Linear}
    boneToModify {boneName=dyn_hair_bk_00_l_JNT}
    lookAtLocation {x=0}
    lookAtTarget {boneReference {boneName=dyn_hair_bk_00_l_JNT_aim}}
    lookAt_Axis {axis=<1 0>}
    lookUp_Axis {axis {x=0, z=1}}
   }
   n5 = Look At - Bone: dyn_hair_bk_00_JNT(ComponentPose=@n4, Alpha=@n2) {
    alphaBoolBlend {blendOption=Linear}
    boneToModify {boneName=dyn_hair_bk_00_JNT}
    lookAtLocation {x=0}
    lookAtTarget {boneReference {boneName=dyn_hair_bk_00_JNT_aim}}
    lookAt_Axis {axis=<1 0>}
    lookUp_Axis {axis {x=0, z=1}}
   }
   n6 = Look At - Bone: dyn_hair_bk_00_r_JNT(ComponentPose=@n5, Alpha=@n2) {
    alphaBoolBlend {blendOption=Linear}
    boneToModify {boneName=dyn_hair_bk_00_r_JNT}
    lookAtLocation {x=0}
    lookAtTarget {boneReference {boneName=dyn_hair_bk_00_r_JNT_aim}}
    lookAt_Axis {axis=<1 0>}
    lookUp_Axis {axis {x=0, z=1}}
   }
   n7 = Look At - Bone: dyn_hair_md_00_r_JNT(ComponentPose=@n6, Alpha=@n2) {
    alphaBoolBlend {blendOption=Linear}
    boneToModify {boneName=dyn_hair_md_00_r_JNT}
    lookAtLocation {x=0}
    lookAtTarget {boneReference {boneName=dyn_hair_md_00_r_JNT_aim}}
    lookAt_Axis {axis=<1 0>}
    lookUp_Axis {axis {x=0, z=1}}
   }
   n8 = RigidBody!(ComponentPose=@n7, Alpha, OverrideWorldGravity:Vector=Get windVector() -> (windVector:Vector), ExternalForce) -> (Pose) {
    bClampLinearTranslationLimitToRefPose=true
    bForceDisableCollisionBetweenConstraintBodies=true
    bOverrideWorldGravity=true
    overridePhysicsAsset=&/F_MED_Constellation_Star_FaceAcc_Physics
    simulationSpace=BaseBoneSpace
    alphaBoolBlend {blendOption=Linear}
    baseBoneRef {boneName=neck_01}
    componentAppliedLinearAccClamp=<2000 2000 2000>
    componentLinearAccScale=<0.4 0.3 0.3>
    componentLinearVelScale=<0.4 0.3 0.3>
    overrideWorldGravity {z=-980}
   }
   n9 = "Transform (Modify) Bone - Bone: dyn_hair_bk_00_l_MSTR"(ComponentPose=@n8, Alpha=1, Translation="0,0,0.3", Rotation="0,0,0", Scale="1,1,1") {
    translationMode=BMM_Additive
    translationSpace=BCS_BoneSpace
    alphaBoolBlend {blendOption=Linear}
    alphaScaleBiasClamp {bInterpResult=true}
    boneToModify {boneName=dyn_hair_bk_00_l_MSTR}
    translation {z=0.3}
   }
   Output Pose(Result=Component To Local(ComponentPose=@n9)) {layerGroup=None}
  ]}
 EventGraph {nodes [
   n1 = Event BlueprintUpdateAnimation()
   > n2 = Sequence()
   n3 = Break Fort Anim Input Skydiving External Force(FortAnimInput_SkydivingExternalForce=Get SkydivingExternalForceInput()) -> (
    bUseSkydivingVectorForce=~
    SkydivingVectorForce=~
    HeadToPelvisDirection=~
    FloatingMultiplier=~
    FloatingAdditive=~
    DivingMultiplier=~
    DivingAdditive=~
    ParachutingMultiplier=~
    ParachutingAdditive=~
   )
   > Set windVector(windVector:Vector=SelectVector(A=vector * float(A=@n3.SkydivingVectorForce, B=0.15), B="0, 0,-980", bPickA=Get bIsSkyDiving())) -> (Output_Get:Vector)
   n4 = NOT Boolean(A=Get bIsPlayingEmote())
   Set LookAtAlpha(
    execute=@n2.then_1
    LookAtAlpha:double=FInterpTo(Current=Get LookAtAlpha(), Target="To Float (Boolean)"(InBool=@n4), DeltaTime=@n1.DeltaTimeX, InterpSpeed=7)
   ) -> (Output_Get:double)
  ]}
]
variables [
 LookAtAlpha {type=number, default=1}
 windVector {type=Vector}
]
